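# Physics asset for the Galileo Outrigger male medium player body cosmetic, set up for four dangling pouches, two on each side.
# Pelvis and thigh capsules give the lower body collision volumes, while each pouch is a box paired with a tiny capsule on its base bone.
# Each pouch is constrained to its base with limited swing and tight twist; bases and thighs join the pelvis with free angular motion.
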PhysicsAsset /BRCosmetics/Characters/Player/Male/Medium/Bodies/M_MED_Galileo_Outrigger/Meshes/M_MED_Galileo_Outrigger_Physics
bodies {
 dyn_pouch_a_base_l [{shapeType=Capsule, length=0.505, radius=0.505, rotation=<rot 0 -90 0>}]
 dyn_pouch_a_base_r [{shapeType=Capsule, length=0.505, radius=0.505, rotation=<rot 0 -90 0>}]
 dyn_pouch_a_l [{shapeType=Box, extentX=11, extentY=7.5, extentZ=4.5, center=<5.0406 0 2.1768>}]
 dyn_pouch_a_r [{shapeType=Box, extentX=9.5, extentY=7, extentZ=2.5, center=<4.7035 0.1447 1.1157>}]
 dyn_pouch_b_base_l [{shapeType=Capsule, length=0.505, radius=0.505, rotation=<rot 0 -90 0>}]
 dyn_pouch_b_base_r [{shapeType=Capsule, length=0.505, radius=0.505, rotation=<rot 0 -90 0>}]
 dyn_pouch_b_l [{shapeType=Box, extentX=14.5, extentY=7.5, extentZ=5.5, center=<6.4698 -0.0001 3.1322>}]
 dyn_pouch_b_r [{shapeType=Box, extentX=12.5, extentY=7, extentZ=4, center=<6.1774 0.0757 1.7861>}]
 pelvis [{shapeType=Capsule, length=5.5, radius=13.75, center=<-5.0566 -2.1587 0>}]
 thigh_l [{shapeType=Capsule, length=27.5, radius=8.75, center=<-16.5892 0.0003 -0.1476>, rotation=<rot 90 0 0>}]
 thigh_r [{shapeType=Capsule, length=27.5, radius=8.75, center=<16.5892 -0.0003 0.1476>, rotation=<rot 90 0 180>}]
}
constraints [
 {bone1Name=dyn_pouch_a_l, bone2Name=dyn_pouch_a_base_l, swing1LimitDegrees=25, swing2LimitDegrees=30, twistLimitDegrees=5}
 {bone1Name=dyn_pouch_a_r, bone2Name=dyn_pouch_a_base_r, swing1LimitDegrees=25, swing2LimitDegrees=30, twistLimitDegrees=5}
 {bone1Name=dyn_pouch_b_r, bone2Name=dyn_pouch_b_base_r, swing1LimitDegrees=25, swing2LimitDegrees=30, twistLimitDegrees=5}
 {bone1Name=dyn_pouch_b_l, bone2Name=dyn_pouch_b_base_l, swing1LimitDegrees=25, swing2LimitDegrees=30, twistLimitDegrees=5}
 {
  bone1Name=thigh_l
  bone2Name=pelvis
  swing1LimitDegrees=45
  swing1Motion=Free
  swing2LimitDegrees=45
  swing2Motion=Free
  twistLimitDegrees=45
  twistMotion=Free
 }
 {
  bone1Name=thigh_r
  bone2Name=pelvis
  swing1LimitDegrees=45
  swing1Motion=Free
  swing2LimitDegrees=45
  swing2Motion=Free
  twistLimitDegrees=45
  twistMotion=Free
 }
 {
  bone1Name=dyn_pouch_a_base_l
  bone2Name=pelvis
  swing1LimitDegrees=45
  swing1Motion=Free
  swing2LimitDegrees=45
  swing2Motion=Free
  twistLimitDegrees=45
  twistMotion=Free
 }
 {
  bone1Name=dyn_pouch_a_base_r
  bone2Name=pelvis
  swing1LimitDegrees=45
  swing1Motion=Free
  swing2LimitDegrees=45
  swing2Motion=Free
  twistLimitDegrees=45
  twistMotion=Free
 }
 {
  bone1Name=dyn_pouch_b_base_r
  bone2Name=pelvis
  swing1LimitDegrees=45
  swing1Motion=Free
  swing2LimitDegrees=45
  swing2Motion=Free
  twistLimitDegrees=45
  twistMotion=Free
 }
 {
  bone1Name=dyn_pouch_b_base_l
  bone2Name=pelvis
  swing1LimitDegrees=45
  swing1Motion=Free
  swing2LimitDegrees=45
  swing2Motion=Free
  twistLimitDegrees=45
  twistMotion=Free
 }
]
