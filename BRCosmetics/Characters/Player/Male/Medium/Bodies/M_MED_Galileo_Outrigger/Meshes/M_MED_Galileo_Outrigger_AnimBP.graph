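# Secondary-motion AnimBP for the male medium Galileo Outrigger body, copying the attached parent mesh's pose.
# RBF Pose Drivers on both upper arms drive shoulder pads, then Look Ats aim armor at the forearms.
# Spine-weighted constraints position chest armor, while additive bone offsets, many scaled by limb angles or crouch alpha, adjust armor, pouches and crouch pad.
# Pelvis-space RigidBody simulation finishes; shoulder master constraints and some shoulder offsets are unconnected.
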AnimBlueprint /BRCosmetics/Characters/Player/Male/Medium/Bodies/M_MED_Galileo_Outrigger/Meshes/M_MED_Galileo_Outrigger_AnimBP
target_skeleton=/Game/Characters/Player/Male/Medium/Bodies/M_MED_Galileo_Outrigger/Meshes/M_MED_Galileo_Outrigger_Skeleton
graphs [
 AnimGraph {nodes [
   "Constraint - dyn_shoulder_armor_r_MSTR (PPPP)"!(
    ComponentPose
    Alpha=1
    ConstraintWeights_0=0.7506
    ConstraintWeights_1:float=0.0856
    ConstraintWeights_2:float=0.1366
    ConstraintWeights_3:float=0.0272
   ) -> (Pose) {
    alphaBoolBlend {blendOption=Linear}
    boneToModify {boneName=dyn_shoulder_armor_r_MSTR}
    constraintSetup [
     {targetBone {boneName=dyn_shoulder_strap_r}}
     {targetBone {boneName=dyn_chest_armor_bk_end_r}}
     {targetBone {boneName=dyn_chest_armor_fr_end_r}}
     {targetBone {boneName=spine_05}}
    ]
    constraintWeights [1,1,1,1]
   }
   "Constraint - dyn_shoulder_armor_l_MSTR (PPPP)"!(
    ComponentPose
    Alpha=1
    ConstraintWeights_0=0.0856
    ConstraintWeights_1:float=0.0272
    ConstraintWeights_2:float=0.7506
    ConstraintWeights_3:float=0.1366
   ) -> (Pose) {
    alphaBoolBlend {blendOption=Linear}
    boneToModify {boneName=dyn_shoulder_armor_l_MSTR}
    constraintSetup [
     {targetBone {boneName=dyn_chest_armor_bk_end_l}}
     {targetBone {boneName=spine_05}}
     {targetBone {boneName=dyn_shoulder_strap_l}}
     {targetBone {boneName=dyn_chest_armor_fr_end_l}}
    ]
    constraintWeights [1,1,1,1]
   }
   "Transform (Modify) Bone - Bone: dyn_shoulder_armor_l_Aim"(Alpha=Get ShoulderLeftPitch() -> (ShoulderLeftPitch:float), Translation="1,-1,1", Rotation="0,0,0", Scale="1,1,1") {
    rotationMode=BMM_Additive
    rotationSpace=BCS_BoneSpace
    translationMode=BMM_Additive
    translationSpace=BCS_BoneSpace
    alphaBoolBlend {blendOption=Linear}
    alphaScaleBiasClamp {bMapRange=true, inRange {max=40}}
    boneToModify {boneName=dyn_shoulder_armor_l_Aim}
    translation=<1 -1 1>
   }
   "Transform (Modify) Bone - Bone: dyn_shoulder_armor_r_Aim"(Alpha=Get UpperArmRightYaw() -> (UpperArmRightYaw:float), Translation="1,1.5,-1", Rotation="0,0,0", Scale="1,1,1") {
    rotationMode=BMM_Additive
    rotationSpace=BCS_BoneSpace
    translationMode=BMM_Additive
    translationSpace=BCS_BoneSpace
    alphaBoolBlend {blendOption=Linear}
    alphaScaleBiasClamp {bMapRange=true, inRange {max=120}}
    boneToModify {boneName=dyn_shoulder_armor_r_Aim}
    translation=<1 1.5 -1>
   }
   "Transform (Modify) Bone - Bone: dyn_shoulder_armor_r_Aim"(Alpha=Get ShoulderRightPitch() -> (ShoulderRightPitch:float), Translation="-1,1,-1", Rotation="0,0,0", Scale="1,1,1") {
    rotationMode=BMM_Additive
    rotationSpace=BCS_BoneSpace
    translationMode=BMM_Additive
    translationSpace=BCS_BoneSpace
    alphaBoolBlend {blendOption=Linear}
    alphaScaleBiasClamp {bMapRange=true, inRange {max=40}}
    boneToModify {boneName=dyn_shoulder_armor_r_Aim}
    translation=<-1 1 -1>
   }
   n1 = Pose Driver - Source: upperarm_l!(SourcePose:Pose Link Structure=Copy Pose From Mesh() {bUseAttachedParent=true}) -> (Pose:Pose Link Structure) {
    is=title AnimGraphNode_PoseDriver
    poseAsset=&/M_MED_Galileo_Outrigger_Shldrs_RBF_ShldrPad_ROM_poses
    onlyDriveBones [{boneName=dyn_shoulder_armor_l_RBF}]
    poseTargets [
     {
      bApplyCustomCurve=false
      bIsHidden=false
      distanceMethod=DefaultMethod
      drivenName=Pose_84
      functionType=DefaultFunction
      targetScale=0.9926
      boneTransforms [{targetRotation=<rot -37.7574 0.1509 2.6875>, targetTranslation=<15.5111 0 0>}]
      customCurve {defaultValue=3.4028234663852886e+38, postInfinityExtrap=RCCE_Constant, preInfinityExtrap=RCCE_Constant, keys []}
      targetRotation=<rot 0 0 0>
     }
     {
      bApplyCustomCurve=false
      bIsHidden=false
      distanceMethod=DefaultMethod
      drivenName=Pose_85
      functionType=DefaultFunction
      targetScale=0.8493
      boneTransforms [{targetRotation=<rot 7.6449 2.0825 2.1436>, targetTranslation=<15.5111 0 0>}]
      customCurve {defaultValue=3.4028234663852886e+38, postInfinityExtrap=RCCE_Constant, preInfinityExtrap=RCCE_Constant, keys []}
      targetRotation=<rot 0 0 0>
     }
     {
      bApplyCustomCurve=false
      bIsHidden=false
      distanceMethod=DefaultMethod
      drivenName=Pose_86
      functionType=DefaultFunction
      targetScale=0.9739
      boneTransforms [{targetRotation=<rot 57.226 5.1005 3.9268>, targetTranslation=<15.5111 0 0>}]
      customCurve {defaultValue=3.4028234663852886e+38, postInfinityExtrap=RCCE_Constant, preInfinityExtrap=RCCE_Constant, keys []}
      targetRotation=<rot 0 0 0>
     }
     {
      bApplyCustomCurve=false
      bIsHidden=false
      distanceMethod=DefaultMethod
      drivenName=Pose_87
      functionType=DefaultFunction
      targetScale=0.5974
      boneTransforms [{targetRotation=<rot 28.0414 52.3966 7.0808>, targetTranslation=<15.5111 0 0>}]
      customCurve {defaultValue=3.4028234663852886e+38, postInfinityExtrap=RCCE_Constant, preInfinityExtrap=RCCE_Constant, keys []}
      targetRotation=<rot 0 0 0>
     }
     {
      bApplyCustomCurve=false
      bIsHidden=false
      distanceMethod=DefaultMethod
      drivenName=Pose_88
      functionType=DefaultFunction
      targetScale=0.709
      boneTransforms [{targetRotation=<rot -2.3017 93.1218 7.599>, targetTranslation=<15.5111 0 0>}]
      customCurve {defaultValue=3.4028234663852886e+38, postInfinityExtrap=RCCE_Constant, preInfinityExtrap=RCCE_Constant, keys []}
      targetRotation=<rot 0 0 0>
     }
     {
      bApplyCustomCurve=false
      bIsHidden=false
      distanceMethod=DefaultMethod
      drivenName=Pose_89
      functionType=DefaultFunction
      targetScale=0.58
      boneTransforms [{targetRotation=<rot 3.654 49.1104 7.0518>, targetTranslation=<15.5111 0 0>}]
      customCurve {defaultValue=3.4028234663852886e+38, postInfinityExtrap=RCCE_Constant, preInfinityExtrap=RCCE_Constant, keys []}
      targetRotation=<rot 0 0 0>
     }
     {
      bApplyCustomCurve=false
      bIsHidden=false
      distanceMethod=DefaultMethod
      drivenName=Pose_90
      functionType=DefaultFunction
      targetScale=0.7046
      boneTransforms [{targetRotation=<rot 5.719 -33.0659 -2.0447>, targetTranslation=<15.5111 0 0>}]
      customCurve {defaultValue=3.4028234663852886e+38, postInfinityExtrap=RCCE_Constant, preInfinityExtrap=RCCE_Constant, keys []}
      targetRotation=<rot 0 0 0>
     }
     {
      bApplyCustomCurve=false
      bIsHidden=false
      distanceMethod=DefaultMethod
      drivenName=Pose_91
      functionType=DefaultFunction
      targetScale=0.7046
      boneTransforms [{targetRotation=<rot 5.272 -62.1971 -5.9428>, targetTranslation=<15.5111 0 0>}]
      customCurve {defaultValue=3.4028234663852886e+38, postInfinityExtrap=RCCE_Constant, preInfinityExtrap=RCCE_Constant, keys []}
      targetRotation=<rot 0 0 0>
     }
     {
      bApplyCustomCurve=false
      bIsHidden=false
      distanceMethod=DefaultMethod
      drivenName=Pose_92
      functionType=DefaultFunction
      targetScale=0.6883
      boneTransforms [{targetRotation=<rot -22.8671 -46.383 9.8347>, targetTranslation=<15.5111 0 0>}]
      customCurve {defaultValue=3.4028234663852886e+38, postInfinityExtrap=RCCE_Constant, preInfinityExtrap=RCCE_Constant, keys []}
      targetRotation=<rot 0 0 0>
     }
     {
      bApplyCustomCurve=false
      bIsHidden=false
      distanceMethod=DefaultMethod
      drivenName=Pose_93
      functionType=DefaultFunction
      targetScale=0.7716
      boneTransforms [{targetRotation=<rot -78.4439 -8.6556 10.6642>, targetTranslation=<15.5111 0 0>}]
      customCurve {defaultValue=3.4028234663852886e+38, postInfinityExtrap=RCCE_Constant, preInfinityExtrap=RCCE_Constant, keys []}
      targetRotation=<rot 0 0 0>
     }
     {
      bApplyCustomCurve=false
      bIsHidden=false
      distanceMethod=DefaultMethod
      drivenName=Pose_94
      functionType=DefaultFunction
      targetScale=0.726
      boneTransforms [{targetRotation=<rot -41.935 55.3657 -40.6602>, targetTranslation=<15.5111 0 0>}]
      customCurve {defaultValue=3.4028234663852886e+38, postInfinityExtrap=RCCE_Constant, preInfinityExtrap=RCCE_Constant, keys []}
      targetRotation=<rot 0 0 0>
     }
     {
      bApplyCustomCurve=false
      bIsHidden=false
      distanceMethod=DefaultMethod
      drivenName=Pose_95
      functionType=DefaultFunction
      targetScale=0.6883
      boneTransforms [{targetRotation=<rot -43.4624 -69.9666 69.4913>, targetTranslation=<15.5111 0 0>}]
      customCurve {defaultValue=3.4028234663852886e+38, postInfinityExtrap=RCCE_Constant, preInfinityExtrap=RCCE_Constant, keys []}
      targetRotation=<rot 0 0 0>
     }
     {
      bApplyCustomCurve=false
      bIsHidden=false
      distanceMethod=DefaultMethod
      drivenName=Pose_96
      functionType=DefaultFunction
      targetScale=0.6159
      boneTransforms [{targetRotation=<rot 48.0922 72.4313 -6.6698>, targetTranslation=<15.5111 0 0>}]
      customCurve {defaultValue=3.4028234663852886e+38, postInfinityExtrap=RCCE_Constant, preInfinityExtrap=RCCE_Constant, keys []}
      targetRotation=<rot 0 0 0>
     }
     {
      bApplyCustomCurve=false
      bIsHidden=false
      distanceMethod=DefaultMethod
      drivenName=Pose_97
      functionType=DefaultFunction
      targetScale=0.7716
      boneTransforms [{targetRotation=<rot -69.4979 -172.5092 173.9242>, targetTranslation=<15.5111 0 0>}]
      customCurve {defaultValue=3.4028234663852886e+38, postInfinityExtrap=RCCE_Constant, preInfinityExtrap=RCCE_Constant, keys []}
      targetRotation=<rot 0 0 0>
     }
     {
      bApplyCustomCurve=false
      bIsHidden=false
      distanceMethod=DefaultMethod
      drivenName=Pose_98
      functionType=DefaultFunction
      targetScale=0.58
      boneTransforms [{targetRotation=<rot -13.4769 65.8568 -31.4924>, targetTranslation=<15.5111 0 0>}]
      customCurve {defaultValue=3.4028234663852886e+38, postInfinityExtrap=RCCE_Constant, preInfinityExtrap=RCCE_Constant, keys []}
      targetRotation=<rot 0 0 0>
     }
     {
      bApplyCustomCurve=false
      bIsHidden=false
      distanceMethod=DefaultMethod
      drivenName=Pose_99
      functionType=DefaultFunction
      targetScale=0.9065
      boneTransforms [{targetRotation=<rot -53.1476 109.6244 -57.8883>, targetTranslation=<15.5111 0 0>}]
      customCurve {defaultValue=3.4028234663852886e+38, postInfinityExtrap=RCCE_Constant, preInfinityExtrap=RCCE_Constant, keys []}
      targetRotation=<rot 0 0 0>
     }
     {
      bApplyCustomCurve=false
      bIsHidden=false
      distanceMethod=DefaultMethod
      drivenName=Pose_100
      functionType=DefaultFunction
      targetScale=0.921
      boneTransforms [{targetRotation=<rot 32.4961 118.7317 28.9677>, targetTranslation=<15.5111 0 0>}]
      customCurve {defaultValue=3.4028234663852886e+38, postInfinityExtrap=RCCE_Constant, preInfinityExtrap=RCCE_Constant, keys []}
      targetRotation=<rot 0 0 0>
     }
     {
      bApplyCustomCurve=false
      bIsHidden=false
      distanceMethod=DefaultMethod
      drivenName=Pose_101
      functionType=DefaultFunction
      targetScale=0.835
      boneTransforms [{targetRotation=<rot 39.4008 -57.7573 -45.6614>, targetTranslation=<15.5111 0 0>}]
      customCurve {defaultValue=3.4028234663852886e+38, postInfinityExtrap=RCCE_Constant, preInfinityExtrap=RCCE_Constant, keys []}
      targetRotation=<rot 0 0 0>
     }
     {
      bApplyCustomCurve=false
      bIsHidden=false
      distanceMethod=DefaultMethod
      drivenName=Pose_102
      functionType=DefaultFunction
      targetScale=1
      boneTransforms [{targetRotation=<rot -6.7303 134.1828 4.2185>, targetTranslation=<15.5111 0 0>}]
      customCurve {defaultValue=3.4028234663852886e+38, postInfinityExtrap=RCCE_Constant, preInfinityExtrap=RCCE_Constant, keys []}
      targetRotation=<rot 0 0 0>
     }
    ]
    rBFParams {radius=35}
    sourceBones [{boneName=upperarm_l}]
   }
   n2 = Pose Driver - Source: upperarm_r!(SourcePose:Pose Link Structure=@n1) -> (Pose:Pose Link Structure) {
    is=title AnimGraphNode_PoseDriver
    poseAsset=&M_MED_Galileo_Outrigger_Shldrs_RBF_ShldrPad_ROM_poses
    onlyDriveBones [{boneName=dyn_shoulder_armor_r_RBF}]
    poseTargets [
     {
      bApplyCustomCurve=false
      bIsHidden=false
      distanceMethod=DefaultMethod
      drivenName=Pose_84
      functionType=DefaultFunction
      targetScale=0.9926
      boneTransforms [{targetRotation=<rot -37.7574 0.1509 2.6875>, targetTranslation=<-15.5111 0 0>}]
      customCurve {defaultValue=3.4028234663852886e+38, postInfinityExtrap=RCCE_Constant, preInfinityExtrap=RCCE_Constant, keys []}
      targetRotation=<rot 0 0 0>
     }
     {
      bApplyCustomCurve=false
      bIsHidden=false
      distanceMethod=DefaultMethod
      drivenName=Pose_85
      functionType=DefaultFunction
      targetScale=0.8493
      boneTransforms [{targetRotation=<rot 7.6449 2.0825 2.1436>, targetTranslation=<-15.5111 0 0>}]
      customCurve {defaultValue=3.4028234663852886e+38, postInfinityExtrap=RCCE_Constant, preInfinityExtrap=RCCE_Constant, keys []}
      targetRotation=<rot 0 0 0>
     }
     {
      bApplyCustomCurve=false
      bIsHidden=false
      distanceMethod=DefaultMethod
      drivenName=Pose_86
      functionType=DefaultFunction
      targetScale=0.9739
      boneTransforms [{targetRotation=<rot 57.226 5.1005 3.9268>, targetTranslation=<-15.5111 0 0>}]
      customCurve {defaultValue=3.4028234663852886e+38, postInfinityExtrap=RCCE_Constant, preInfinityExtrap=RCCE_Constant, keys []}
      targetRotation=<rot 0 0 0>
     }
     {
      bApplyCustomCurve=false
      bIsHidden=false
      distanceMethod=DefaultMethod
      drivenName=Pose_87
      functionType=DefaultFunction
      targetScale=0.5974
      boneTransforms [{targetRotation=<rot 28.0414 52.3966 7.0808>, targetTranslation=<-15.5111 0 0>}]
      customCurve {defaultValue=3.4028234663852886e+38, postInfinityExtrap=RCCE_Constant, preInfinityExtrap=RCCE_Constant, keys []}
      targetRotation=<rot 0 0 0>
     }
     {
      bApplyCustomCurve=false
      bIsHidden=false
      distanceMethod=DefaultMethod
      drivenName=Pose_88
      functionType=DefaultFunction
      targetScale=0.709
      boneTransforms [{targetRotation=<rot -2.3017 93.1218 7.599>, targetTranslation=<-15.5111 0 0>}]
      customCurve {defaultValue=3.4028234663852886e+38, postInfinityExtrap=RCCE_Constant, preInfinityExtrap=RCCE_Constant, keys []}
      targetRotation=<rot 0 0 0>
     }
     {
      bApplyCustomCurve=false
      bIsHidden=false
      distanceMethod=DefaultMethod
      drivenName=Pose_89
      functionType=DefaultFunction
      targetScale=0.58
      boneTransforms [{targetRotation=<rot 3.654 49.1104 7.0518>, targetTranslation=<-15.5111 0 0>}]
      customCurve {defaultValue=3.4028234663852886e+38, postInfinityExtrap=RCCE_Constant, preInfinityExtrap=RCCE_Constant, keys []}
      targetRotation=<rot 0 0 0>
     }
     {
      bApplyCustomCurve=false
      bIsHidden=false
      distanceMethod=DefaultMethod
      drivenName=Pose_90
      functionType=DefaultFunction
      targetScale=0.7046
      boneTransforms [{targetRotation=<rot 5.719 -33.0659 -2.0447>, targetTranslation=<-15.5111 0 0>}]
      customCurve {defaultValue=3.4028234663852886e+38, postInfinityExtrap=RCCE_Constant, preInfinityExtrap=RCCE_Constant, keys []}
      targetRotation=<rot 0 0 0>
     }
     {
      bApplyCustomCurve=false
      bIsHidden=false
      distanceMethod=DefaultMethod
      drivenName=Pose_91
      functionType=DefaultFunction
      targetScale=0.7046
      boneTransforms [{targetRotation=<rot 5.272 -62.1971 -5.9428>, targetTranslation=<-15.5111 0 0>}]
      customCurve {defaultValue=3.4028234663852886e+38, postInfinityExtrap=RCCE_Constant, preInfinityExtrap=RCCE_Constant, keys []}
      targetRotation=<rot 0 0 0>
     }
     {
      bApplyCustomCurve=false
      bIsHidden=false
      distanceMethod=DefaultMethod
      drivenName=Pose_92
      functionType=DefaultFunction
      targetScale=0.6883
      boneTransforms [{targetRotation=<rot -22.8671 -46.383 9.8347>, targetTranslation=<-15.5111 0 0>}]
      customCurve {defaultValue=3.4028234663852886e+38, postInfinityExtrap=RCCE_Constant, preInfinityExtrap=RCCE_Constant, keys []}
      targetRotation=<rot 0 0 0>
     }
     {
      bApplyCustomCurve=false
      bIsHidden=false
      distanceMethod=DefaultMethod
      drivenName=Pose_93
      functionType=DefaultFunction
      targetScale=0.7716
      boneTransforms [{targetRotation=<rot -78.4439 -8.6556 10.6642>, targetTranslation=<-15.5111 0 0>}]
      customCurve {defaultValue=3.4028234663852886e+38, postInfinityExtrap=RCCE_Constant, preInfinityExtrap=RCCE_Constant, keys []}
      targetRotation=<rot 0 0 0>
     }
     {
      bApplyCustomCurve=false
      bIsHidden=false
      distanceMethod=DefaultMethod
      drivenName=Pose_94
      functionType=DefaultFunction
      targetScale=0.726
      boneTransforms [{targetRotation=<rot -41.935 55.3657 -40.6602>, targetTranslation=<-15.5111 0 0>}]
      customCurve {defaultValue=3.4028234663852886e+38, postInfinityExtrap=RCCE_Constant, preInfinityExtrap=RCCE_Constant, keys []}
      targetRotation=<rot 0 0 0>
     }
     {
      bApplyCustomCurve=false
      bIsHidden=false
      distanceMethod=DefaultMethod
      drivenName=Pose_95
      functionType=DefaultFunction
      targetScale=0.6883
      boneTransforms [{targetRotation=<rot -43.4624 -69.9666 69.4913>, targetTranslation=<-15.5111 0 0>}]
      customCurve {defaultValue=3.4028234663852886e+38, postInfinityExtrap=RCCE_Constant, preInfinityExtrap=RCCE_Constant, keys []}
      targetRotation=<rot 0 0 0>
     }
     {
      bApplyCustomCurve=false
      bIsHidden=false
      distanceMethod=DefaultMethod
      drivenName=Pose_96
      functionType=DefaultFunction
      targetScale=0.6159
      boneTransforms [{targetRotation=<rot 48.0922 72.4314 -6.6698>, targetTranslation=<-15.5111 0 0>}]
      customCurve {defaultValue=3.4028234663852886e+38, postInfinityExtrap=RCCE_Constant, preInfinityExtrap=RCCE_Constant, keys []}
      targetRotation=<rot 0 0 0>
     }
     {
      bApplyCustomCurve=false
      bIsHidden=false
      distanceMethod=DefaultMethod
      drivenName=Pose_97
      functionType=DefaultFunction
      targetScale=0.7716
      boneTransforms [{targetRotation=<rot -69.4979 -172.5092 173.9242>, targetTranslation=<-15.5111 0 0>}]
      customCurve {defaultValue=3.4028234663852886e+38, postInfinityExtrap=RCCE_Constant, preInfinityExtrap=RCCE_Constant, keys []}
      targetRotation=<rot 0 0 0>
     }
     {
      bApplyCustomCurve=false
      bIsHidden=false
      distanceMethod=DefaultMethod
      drivenName=Pose_98
      functionType=DefaultFunction
      targetScale=0.58
      boneTransforms [{targetRotation=<rot -13.4769 65.8568 -31.4924>, targetTranslation=<-15.5111 0 0>}]
      customCurve {defaultValue=3.4028234663852886e+38, postInfinityExtrap=RCCE_Constant, preInfinityExtrap=RCCE_Constant, keys []}
      targetRotation=<rot 0 0 0>
     }
     {
      bApplyCustomCurve=false
      bIsHidden=false
      distanceMethod=DefaultMethod
      drivenName=Pose_99
      functionType=DefaultFunction
      targetScale=0.9065
      boneTransforms [{targetRotation=<rot -53.1476 109.6244 -57.8883>, targetTranslation=<-15.5111 0 0>}]
      customCurve {defaultValue=3.4028234663852886e+38, postInfinityExtrap=RCCE_Constant, preInfinityExtrap=RCCE_Constant, keys []}
      targetRotation=<rot 0 0 0>
     }
     {
      bApplyCustomCurve=false
      bIsHidden=false
      distanceMethod=DefaultMethod
      drivenName=Pose_100
      functionType=DefaultFunction
      targetScale=0.921
      boneTransforms [{targetRotation=<rot 32.4961 118.7317 28.9677>, targetTranslation=<-15.5111 0 0>}]
      customCurve {defaultValue=3.4028234663852886e+38, postInfinityExtrap=RCCE_Constant, preInfinityExtrap=RCCE_Constant, keys []}
      targetRotation=<rot 0 0 0>
     }
     {
      bApplyCustomCurve=false
      bIsHidden=false
      distanceMethod=DefaultMethod
      drivenName=Pose_101
      functionType=DefaultFunction
      targetScale=0.835
      boneTransforms [{targetRotation=<rot 39.4008 -57.7573 -45.6614>, targetTranslation=<-15.5111 0 0>}]
      customCurve {defaultValue=3.4028234663852886e+38, postInfinityExtrap=RCCE_Constant, preInfinityExtrap=RCCE_Constant, keys []}
      targetRotation=<rot 0 0 0>
     }
     {
      bApplyCustomCurve=false
      bIsHidden=false
      distanceMethod=DefaultMethod
      drivenName=Pose_102
      functionType=DefaultFunction
      targetScale=1
      boneTransforms [{targetRotation=<rot -6.7303 134.1828 4.2185>, targetTranslation=<-15.5111 0 0>}]
      customCurve {defaultValue=3.4028234663852886e+38, postInfinityExtrap=RCCE_Constant, preInfinityExtrap=RCCE_Constant, keys []}
      targetRotation=<rot 0 0 0>
     }
    ]
    rBFParams {radius=35}
    sourceBones [{boneName=upperarm_r}]
   }
   n3 = Look At - Bone: dyn_shoulder_armor_l_Aim(ComponentPose=Local To Component(LocalPose=@n2), Alpha=0.7) {
    alphaBoolBlend {blendOption=Linear}
    boneToModify {boneName=dyn_shoulder_armor_l_Aim}
    lookAtLocation {x=0}
    lookAtTarget {boneReference {boneName=lowerarm_l}}
    lookAt_Axis {axis=<1 0>}
    lookUp_Axis {axis {x=0, z=1}}
   }
   n4 = Look At - Bone: dyn_shoulder_armor_r_Aim(ComponentPose=@n3, Alpha=0.7) {
    alphaBoolBlend {blendOption=Linear}
    boneToModify {boneName=dyn_shoulder_armor_r_Aim}
    lookAtLocation {x=0}
    lookAtTarget {boneReference {boneName=lowerarm_r}}
    lookAt_Axis {axis=<-1 0>}
    lookUp_Axis {axis {x=0, z=1}}
   }
   n5 = "Transform (Modify) Bone - Bone: dyn_shoulder_armor_r_Aim"(ComponentPose=@n4, Alpha=Get UpperArmRightYaw(), Translation="1,4,-2", Rotation="-10,0,0", Scale="1,1,1") {
    rotationMode=BMM_Additive
    rotationSpace=BCS_BoneSpace
    translationMode=BMM_Additive
    translationSpace=BCS_BoneSpace
    alphaBoolBlend {blendOption=Linear}
    alphaScaleBiasClamp {bMapRange=true, inRange {max=120}}
    boneToModify {boneName=dyn_shoulder_armor_r_Aim}
    rotation {pitch=-10}
    translation=<1 4 -2>
   }
   n6 = "Transform (Modify) Bone - Bone: dyn_shoulder_armor_l_Aim"(ComponentPose=@n5, Alpha=Get UpperArmLeftYaw() -> (UpperArmLeftYaw:float), Translation="-1,-4,2", Rotation="-10,0,0", Scale="1,1,1") {
    rotationMode=BMM_Additive
    rotationSpace=BCS_BoneSpace
    translationMode=BMM_Additive
    translationSpace=BCS_BoneSpace
    alphaBoolBlend {blendOption=Linear}
    alphaScaleBiasClamp {bMapRange=true, inRange {max=120}}
    boneToModify {boneName=dyn_shoulder_armor_l_Aim}
    rotation {pitch=-10}
    translation=<-1 -4 2>
   }
   n7 = "Constraint - dyn_chest_armor_bk_l_MSTR (PPP)"!(ComponentPose=@n6, Alpha=1, ConstraintWeights_0=0.6, ConstraintWeights_1:float=0.4, ConstraintWeights_2:float=0) -> (Pose) {
    alphaBoolBlend {blendOption=Linear}
    boneToModify {boneName=dyn_chest_armor_bk_l_MSTR}
    constraintSetup [{targetBone {boneName=spine_04}}, {targetBone {boneName=spine_03}}, {targetBone {boneName=spine_05}}]
    constraintWeights [1,1,1]
   }
   n8 = "Constraint - dyn_chest_armor_bk_r_MSTR (PPP)"!(ComponentPose=@n7, Alpha=1, ConstraintWeights_0=0.6, ConstraintWeights_1:float=0.4, ConstraintWeights_2:float=0) -> (Pose) {
    alphaBoolBlend {blendOption=Linear}
    boneToModify {boneName=dyn_chest_armor_bk_r_MSTR}
    constraintSetup [{targetBone {boneName=spine_04}}, {targetBone {boneName=spine_03}}, {targetBone {boneName=spine_05}}]
    constraintWeights [1,1,1]
   }
   n9 = "Constraint - dyn_chest_armor_side_l_MSTR (PPP)"!(ComponentPose=@n8, Alpha=1, ConstraintWeights_0=0.6, ConstraintWeights_1:float=0.4, ConstraintWeights_2:float=0) -> (Pose) {
    alphaBoolBlend {blendOption=Linear}
    boneToModify {boneName=dyn_chest_armor_side_l_MSTR}
    constraintSetup [{targetBone {boneName=spine_04}}, {targetBone {boneName=spine_03}}, {targetBone {boneName=spine_05}}]
    constraintWeights [1,1,1]
   }
   n10 = "Constraint - dyn_chest_armor_side_r_MSTR (PPP)"!(ComponentPose=@n9, Alpha=1, ConstraintWeights_0=0.6, ConstraintWeights_1:float=0.4, ConstraintWeights_2:float=0) -> (Pose) {
    alphaBoolBlend {blendOption=Linear}
    boneToModify {boneName=dyn_chest_armor_side_r_MSTR}
    constraintSetup [{targetBone {boneName=spine_04}}, {targetBone {boneName=spine_03}}, {targetBone {boneName=spine_05}}]
    constraintWeights [1,1,1]
   }
   n11 = "Constraint - dyn_chest_armor_fr_l_MSTR (PPP)"!(ComponentPose=@n10, Alpha=1, ConstraintWeights_0=0.6, ConstraintWeights_1:float=0.4, ConstraintWeights_2:float=0) -> (Pose) {
    alphaBoolBlend {blendOption=Linear}
    boneToModify {boneName=dyn_chest_armor_fr_l_MSTR}
    constraintSetup [{targetBone {boneName=spine_04}}, {targetBone {boneName=spine_03}}, {targetBone {boneName=spine_05}}]
    constraintWeights [1,1,1]
   }
   n12 = "Constraint - dyn_chest_armor_fr_r_MSTR (PPP)"!(ComponentPose=@n11, Alpha=1, ConstraintWeights_0=0.6, ConstraintWeights_1:float=0.4, ConstraintWeights_2:float=0) -> (Pose) {
    alphaBoolBlend {blendOption=Linear}
    boneToModify {boneName=dyn_chest_armor_fr_r_MSTR}
    constraintSetup [{targetBone {boneName=spine_04}}, {targetBone {boneName=spine_03}}, {targetBone {boneName=spine_05}}]
    constraintWeights [1,1,1]
   }
   n13 = "Transform (Modify) Bone - Bone: dyn_chest_armor_bk_l_MSTR"(ComponentPose=@n12, Alpha=1, Translation="0,0,0", Rotation="0,5,0", Scale="1,1,1") {
    rotationMode=BMM_Additive
    rotationSpace=BCS_BoneSpace
    translationMode=BMM_Additive
    translationSpace=BCS_BoneSpace
    alphaBoolBlend {blendOption=Linear}
    alphaScaleBiasClamp {inRange {max=0, min=-100}, outRange {max=0, min=1}}
    boneToModify {boneName=dyn_chest_armor_bk_l_MSTR}
    rotation {yaw=5}
   }
   n14 = "Transform (Modify) Bone - Bone: dyn_chest_armor_bk_r_MSTR"(ComponentPose=@n13, Alpha=1, Translation="0,0,0", Rotation="0,5,0", Scale="1,1,1") {
    rotationMode=BMM_Additive
    rotationSpace=BCS_BoneSpace
    translationMode=BMM_Additive
    translationSpace=BCS_BoneSpace
    alphaBoolBlend {blendOption=Linear}
    alphaScaleBiasClamp {inRange {max=0, min=-100}, outRange {max=0, min=1}}
    boneToModify {boneName=dyn_chest_armor_bk_r_MSTR}
    rotation {yaw=5}
   }
   n15 = "Transform (Modify) Bone - Bone: dyn_chest_armor_side_l_MSTR"(ComponentPose=@n14, Alpha=1, Translation="0,0,0", Rotation="0,5,0", Scale="1,1,1") {
    rotationMode=BMM_Additive
    rotationSpace=BCS_BoneSpace
    translationMode=BMM_Additive
    translationSpace=BCS_BoneSpace
    alphaBoolBlend {blendOption=Linear}
    alphaScaleBiasClamp {inRange {max=0, min=-100}, outRange {max=0, min=1}}
    boneToModify {boneName=dyn_chest_armor_side_l_MSTR}
    rotation {yaw=5}
   }
   n16 = "Transform (Modify) Bone - Bone: dyn_chest_armor_side_r_MSTR"(ComponentPose=@n15, Alpha=1, Translation="0,0,0", Rotation="0,5,0", Scale="1,1,1") {
    rotationMode=BMM_Additive
    rotationSpace=BCS_BoneSpace
    translationMode=BMM_Additive
    translationSpace=BCS_BoneSpace
    alphaBoolBlend {blendOption=Linear}
    alphaScaleBiasClamp {inRange {max=0, min=-100}, outRange {max=0, min=1}}
    boneToModify {boneName=dyn_chest_armor_side_r_MSTR}
    rotation {yaw=5}
   }
   n17 = Get LegLeftYaw() -> (LegLeftYaw:float)
   n18 = "Transform (Modify) Bone - Bone: dyn_pouch_a_base_l"(ComponentPose=@n16, Alpha=@n17, Translation="0,-8,-2", Rotation="-5,-5,-35", Scale="1,1,1") {
    rotationMode=BMM_Additive
    rotationSpace=BCS_BoneSpace
    translationMode=BMM_Additive
    translationSpace=BCS_BoneSpace
    alphaBoolBlend {blendOption=Linear}
    alphaScaleBiasClamp {bMapRange=true, inRange {max=0, min=-100}, outRange {max=0, min=1}}
    boneToModify {boneName=dyn_pouch_a_base_l}
    rotation=<rot -5 -5 -35>
    translation {y=-8, z=-2}
   }
   n19 = "Transform (Modify) Bone - Bone: dyn_pouch_b_base_l"(ComponentPose=@n18, Alpha=@n17, Translation="0,-6.5,-1", Rotation="-10,0,-20", Scale="1,1,1") {
    rotationMode=BMM_Additive
    rotationSpace=BCS_BoneSpace
    translationMode=BMM_Additive
    translationSpace=BCS_BoneSpace
    alphaBoolBlend {blendOption=Linear}
    alphaScaleBiasClamp {bMapRange=true, inRange {max=0, min=-100}, outRange {max=0, min=1}}
    boneToModify {boneName=dyn_pouch_b_base_l}
    rotation {pitch=-10, roll=-20}
    translation {y=-6.5, z=-1}
   }
   n20 = "Transform (Modify) Bone - Bone: dyn_pouch_a_base_l"(ComponentPose=@n19, Alpha=Get LegLeftPitch() -> (LegLeftPitch:float), Translation="0,0,0", Rotation="20,0,0", Scale="1,1,1") {
    rotationMode=BMM_Additive
    rotationSpace=BCS_BoneSpace
    translationMode=BMM_Additive
    translationSpace=BCS_BoneSpace
    alphaBoolBlend {blendOption=Linear}
    alphaScaleBiasClamp {bMapRange=true, inRange {max=20}}
    boneToModify {boneName=dyn_pouch_a_base_l}
    rotation {pitch=20}
   }
   n21 = Get LegRightYaw() -> (LegRightYaw:float)
   n22 = "Transform (Modify) Bone - Bone: dyn_pouch_a_base_r"(ComponentPose=@n20, Alpha=@n21, Translation="-1,6,-2", Rotation="5,10,35", Scale="1,1,1") {
    rotationMode=BMM_Additive
    rotationSpace=BCS_BoneSpace
    translationMode=BMM_Additive
    translationSpace=BCS_BoneSpace
    alphaBoolBlend {blendOption=Linear}
    alphaScaleBiasClamp {bMapRange=true, inRange {max=0, min=-100}, outRange {max=0, min=1}}
    boneToModify {boneName=dyn_pouch_a_base_r}
    rotation=<rot 5 10 35>
    translation=<-1 6 -2>
   }
   n23 = "Transform (Modify) Bone - Bone: dyn_pouch_b_base_r"(ComponentPose=@n22, Alpha=@n21, Translation="-1,6,-0.5", Rotation="0.000,5,25", Scale="1,1,1") {
    rotationMode=BMM_Additive
    rotationSpace=BCS_BoneSpace
    translationMode=BMM_Additive
    translationSpace=BCS_BoneSpace
    alphaBoolBlend {blendOption=Linear}
    alphaScaleBiasClamp {bMapRange=true, inRange {max=0, min=-100}, outRange {max=0, min=1}}
    boneToModify {boneName=dyn_pouch_b_base_r}
    rotation {roll=25, yaw=5}
    translation=<-1 6 -0.5>
   }
   n24 = "Transform (Modify) Bone - Bone: dyn_pouch_a_base_r"(ComponentPose=@n23, Alpha=Get LegRightPitch() -> (LegRightPitch:float), Translation="0,0,0", Rotation="20,0,0", Scale="1,1,1") {
    rotationMode=BMM_Additive
    rotationSpace=BCS_BoneSpace
    translationMode=BMM_Additive
    translationSpace=BCS_BoneSpace
    alphaBoolBlend {blendOption=Linear}
    alphaScaleBiasClamp {bMapRange=true, inRange {max=20}}
    boneToModify {boneName=dyn_pouch_a_base_r}
    rotation {pitch=20}
   }
   n25 = "Transform (Modify) Bone - Bone: dyn_crouchPad"(ComponentPose=@n24, Alpha=Get CrouchingAlpha() -> (CrouchingAlpha:float), Translation="0,0,0", Rotation="0,-20,0", Scale="1,1,1") {
    rotationMode=BMM_Additive
    rotationSpace=BCS_BoneSpace
    translationMode=BMM_Additive
    translationSpace=BCS_BoneSpace
    alphaBoolBlend {blendOption=Linear}
    alphaScaleBiasClamp {inRange {max=0, min=-100}, outRange {max=0, min=1}}
    boneToModify {boneName=dyn_crouchPad}
    rotation {yaw=-20}
   }
   n26 = RigidBody(ComponentPose=@n25) {
    bClampLinearTranslationLimitToRefPose=true
    bOverrideWorldGravity=true
    overridePhysicsAsset=&/M_MED_Galileo_Outrigger_Physics
    simulationSpace=BaseBoneSpace
    alphaBoolBlend {blendOption=Linear}
    baseBoneRef {boneName=pelvis}
    componentLinearAccScale=<0.5 0.5 0.5>
    componentLinearVelScale=<0.5 0.5 0.5>
    overrideWorldGravity {z=-980}
   }
   Output Pose(Result=Component To Local(ComponentPose=@n26)) {layerGroup=None}
  ]}
 EventGraph {nodes [TryGetPawnOwner()]}
]
variables []
